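# Physics asset for the dynamic collar on the female medium Wasteland Warrior Dark player body.
# Five collar segments (back, back-left, back-right, left, right) each pair a capsule body with a small MSTR anchor body through identical angular limits: very tight twist, narrow swing on one axis, wider on the other.
# Capsules on the head, neck and upper arms add body volumes around the collar, with only head and neck joined, by a free-rotation constraint.
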PhysicsAsset /BRCosmetics/Characters/Player/Female/Medium/Bodies/F_MED_Wasteland_Warrior_Dark/Meshes/F_MED_Wasteland_Warrior_Dark_dyn_collar
bodies {
 dyn_collar_bk_01 [{shapeType=Capsule, length=11.5, radius=2, center=<5.045 0 0>, rotation=<rot 0 0 90>}]
 dyn_collar_bk_MSTR [{shapeType=Capsule, length=0.505, radius=0.505, rotation=<rot 0 -90 0>}]
 dyn_collar_bk_l_01 [{shapeType=Capsule, length=7, radius=2, center=<4.552 0 0>, rotation=<rot 0 0 90>}]
 dyn_collar_bk_l_MSTR [{shapeType=Capsule, length=0.505, radius=0.505, rotation=<rot 0 -90 0>}]
 dyn_collar_bk_r_01 [{shapeType=Capsule, length=4.75, radius=2, center=<4.3554 0 0>, rotation=<rot 0 0 105.0013>}]
 dyn_collar_bk_r_MSTR [{shapeType=Capsule, length=0.505, radius=0.505, rotation=<rot 0 -90 0>}]
 dyn_collar_l_01 [{shapeType=Capsule, length=7, radius=2, center=<5.6699 -0.0718 -0.0002>, rotation=<rot 0 -5.0002 90>}]
 dyn_collar_l_MSTR [{shapeType=Capsule, length=0.505, radius=0.505, rotation=<rot 0 -90 0>}]
 dyn_collar_r_01 [{shapeType=Capsule, length=7, radius=2, center=<4.4853 0 0>, rotation=<rot 0 5.0005 94.9999>}]
 dyn_collar_r_MSTR [{shapeType=Capsule, length=0.505, radius=0.505, rotation=<rot 0 -90 0>}]
 head [
  {shapeType=Capsule, length=6.25, radius=5.005, center=<3.0894 2.7125 0>, rotation=<rot 90 0 0>}
  {shapeType=Capsule, length=7.75, radius=2.75, center=<-3.1539 7.1394 0>}
 ]
 neck_01 [{shapeType=Capsule, length=5.75, radius=3.005, center=<1.8097 1.2021 0>, rotation=<rot 90 0 0>}]
 upperarm_l [{shapeType=Capsule, length=11.25, radius=4.005, center=<6.3076 1.5628 1.2526>, rotation=<rot 90 0 0>}]
 upperarm_r [{shapeType=Capsule, length=13, radius=3.255, center=<-6.3076 0.4589 -0.3832>, rotation=<rot 90 0 -179.9999>}]
}
constraints [
 {
  bone1Name=head
  bone2Name=neck_01
  swing1LimitDegrees=45
  swing1Motion=Free
  swing2LimitDegrees=45
  swing2Motion=Free
  twistLimitDegrees=45
  twistMotion=Free
 }
 {bone1Name=dyn_collar_bk_01, bone2Name=dyn_collar_bk_MSTR, swing1LimitDegrees=10, swing2LimitDegrees=45, twistLimitDegrees=2}
 {bone1Name=dyn_collar_bk_r_01, bone2Name=dyn_collar_bk_r_MSTR, swing1LimitDegrees=10, swing2LimitDegrees=45, twistLimitDegrees=2}
 {bone1Name=dyn_collar_r_01, bone2Name=dyn_collar_r_MSTR, swing1LimitDegrees=10, swing2LimitDegrees=45, twistLimitDegrees=2}
 {bone1Name=dyn_collar_bk_l_01, bone2Name=dyn_collar_bk_l_MSTR, swing1LimitDegrees=10, swing2LimitDegrees=45, twistLimitDegrees=2}
 {bone1Name=dyn_collar_l_01, bone2Name=dyn_collar_l_MSTR, swing1LimitDegrees=10, swing2LimitDegrees=45, twistLimitDegrees=2}
]
