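# Physics asset for the dangling straps on the Wasteland Warrior Dark female medium body cosmetic, covering front and back strap bone chains.
# Small master bodies for front, back and back-left groups join the pelvis with free rotation and anchor capsule chains with limited swing and tight twist.
# The front chain splits left and right, and a back-right branch forks off the central back chain.
# Pelvis and thigh capsules form the body volumes beside the straps.
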PhysicsAsset /BRCosmetics/Characters/Player/Female/Medium/Bodies/F_MED_Wasteland_Warrior_Dark/Meshes/F_MED_Wasteland_Warrior_Dark_dyn_straps
bodies {
 dyn_straps_bk_01 [{shapeType=Capsule, length=3.5, radius=3.25, center=<5.514 0 0>, rotation=<rot 0 0 90>}]
 dyn_straps_bk_02 [{shapeType=Capsule, length=3.5, radius=3.25, center=<6 0 0>, rotation=<rot 90 0 0>}]
 dyn_straps_bk_03 [{shapeType=Capsule, length=3.5, radius=3.25, center=<6 0 0>, rotation=<rot 90 0 0>}]
 dyn_straps_bk_04 [{shapeType=Capsule, length=3.5, radius=3.25, center=<6 0 0>, rotation=<rot 90 0 0>}]
 dyn_straps_bk_MSTR [{shapeType=Capsule, length=0.505, radius=0.505, rotation=<rot 0 -90 0>}]
 dyn_straps_bk_l_01 [{shapeType=Capsule, length=3.5, radius=3.25, center=<5.514 0 0>, rotation=<rot 0 0 90>}]
 dyn_straps_bk_l_02 [{shapeType=Capsule, length=3.75, radius=3, center=<6 0 0>, rotation=<rot 90 0 0>}]
 dyn_straps_bk_l_03 [{shapeType=Capsule, length=6, radius=3, center=<6 0 0>, rotation=<rot 90 0 0>}]
 dyn_straps_bk_l_MSTR [{shapeType=Capsule, length=0.505, radius=0.505, rotation=<rot 0 -90 0>}]
 dyn_straps_bk_r_01 [{shapeType=Capsule, length=4, radius=2.75, center=<6 0 0>, rotation=<rot 90 0 0>}]
 dyn_straps_bk_r_02 [{shapeType=Capsule, length=5, radius=2.75, center=<6 0 0>, rotation=<rot 90 0 0>}]
 dyn_straps_ft_01 [{shapeType=Capsule, length=4, radius=2.75, center=<6 0 1.5>, rotation=<rot 90 0 0>}]
 dyn_straps_ft_MSTR [{shapeType=Capsule, length=0.505, radius=0.505, rotation=<rot 0 -90 0>}]
 dyn_straps_ft_l_01 [{shapeType=Capsule, length=4.5, radius=2.25, center=<6 0 0>, rotation=<rot 90 0 0>}]
 dyn_straps_ft_l_02 [{shapeType=Capsule, length=4.5, radius=2.25, center=<6 0 0>, rotation=<rot 90 0 0>}]
 dyn_straps_ft_r_01 [{shapeType=Capsule, length=4.5, radius=2.25, center=<6 0 0>, rotation=<rot 90 0 0>}]
 dyn_straps_ft_r_02 [{shapeType=Capsule, length=4.5, radius=2.25, center=<6 0 0>, rotation=<rot 90 0 0>}]
 pelvis [
  {shapeType=Capsule, length=12.25, radius=7.75, center=<-11.1259 -1.8869 0>, rotation=<rot 90 0 0>}
  {shapeType=Capsule, length=15.892, radius=7.608, center=<-8.0771 -2.1652 0>}
 ]
 thigh_l [{shapeType=Capsule, length=20.75, radius=7.25, center=<-18.4496 0.0005 0.0003>, rotation=<rot 90 0 0>}]
 thigh_r [{shapeType=Capsule, length=20.75, radius=7.25, center=<18.4496 -0.0005 -0.0003>, rotation=<rot 90 0 180>}]
}
constraints [
 {bone1Name=thigh_l, bone2Name=pelvis, swing1LimitDegrees=45, swing2LimitDegrees=45, twistLimitDegrees=45}
 {bone1Name=thigh_r, bone2Name=pelvis, swing1LimitDegrees=45, swing2LimitDegrees=45, twistLimitDegrees=45}
 {
  bone1Name=dyn_straps_ft_MSTR
  bone2Name=pelvis
  swing1LimitDegrees=45
  swing1Motion=Free
  swing2LimitDegrees=45
  swing2Motion=Free
  twistLimitDegrees=45
  twistMotion=Free
 }
 {bone1Name=dyn_straps_ft_01, bone2Name=dyn_straps_ft_MSTR, swing1LimitDegrees=25, swing2LimitDegrees=35, twistLimitDegrees=2}
 {bone1Name=dyn_straps_ft_r_01, bone2Name=dyn_straps_ft_01, swing1LimitDegrees=15, swing2LimitDegrees=25, twistLimitDegrees=2}
 {bone1Name=dyn_straps_ft_r_02, bone2Name=dyn_straps_ft_r_01, swing1LimitDegrees=15, swing2LimitDegrees=25, twistLimitDegrees=2}
 {bone1Name=dyn_straps_ft_l_01, bone2Name=dyn_straps_ft_01, swing1LimitDegrees=15, swing2LimitDegrees=25, twistLimitDegrees=2}
 {bone1Name=dyn_straps_ft_l_02, bone2Name=dyn_straps_ft_l_01, swing1LimitDegrees=15, swing2LimitDegrees=25, twistLimitDegrees=2}
 {
  bone1Name=dyn_straps_bk_MSTR
  bone2Name=pelvis
  swing1LimitDegrees=45
  swing1Motion=Free
  swing2LimitDegrees=45
  swing2Motion=Free
  twistLimitDegrees=45
  twistMotion=Free
 }
 {bone1Name=dyn_straps_bk_01, bone2Name=dyn_straps_bk_MSTR, swing1LimitDegrees=20, swing2LimitDegrees=30, twistLimitDegrees=2}
 {bone1Name=dyn_straps_bk_02, bone2Name=dyn_straps_bk_01, swing1LimitDegrees=45, swing2LimitDegrees=30, twistLimitDegrees=2}
 {bone1Name=dyn_straps_bk_03, bone2Name=dyn_straps_bk_02, swing1LimitDegrees=45, swing2LimitDegrees=30, twistLimitDegrees=2}
 {bone1Name=dyn_straps_bk_04, bone2Name=dyn_straps_bk_03, swing1LimitDegrees=45, swing2LimitDegrees=30, twistLimitDegrees=2}
 {bone1Name=dyn_straps_bk_r_01, bone2Name=dyn_straps_bk_02, swing1LimitDegrees=35, swing2LimitDegrees=30, twistLimitDegrees=2}
 {bone1Name=dyn_straps_bk_r_02, bone2Name=dyn_straps_bk_r_01, swing1LimitDegrees=35, swing2LimitDegrees=30, twistLimitDegrees=2}
 {bone1Name=dyn_straps_bk_l_01, bone2Name=dyn_straps_bk_l_MSTR, swing1LimitDegrees=20, swing2LimitDegrees=30, twistLimitDegrees=2}
 {bone1Name=dyn_straps_bk_l_02, bone2Name=dyn_straps_bk_l_01, swing1LimitDegrees=35, swing2LimitDegrees=30, twistLimitDegrees=2}
 {bone1Name=dyn_straps_bk_l_03, bone2Name=dyn_straps_bk_l_02, swing1LimitDegrees=35, swing2LimitDegrees=30, twistLimitDegrees=2}
 {
  bone1Name=dyn_straps_bk_l_MSTR
  bone2Name=pelvis
  swing1LimitDegrees=45
  swing1Motion=Free
  swing2LimitDegrees=45
  swing2Motion=Free
  twistLimitDegrees=45
  twistMotion=Free
 }
]
