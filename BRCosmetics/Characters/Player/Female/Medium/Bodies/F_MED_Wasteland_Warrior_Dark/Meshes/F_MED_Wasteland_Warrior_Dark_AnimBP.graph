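# Character-part AnimBP for the female Wasteland Warrior Dark body, adding deformation and secondary motion to clothing and armor.
# It copies the parent mesh pose through two Control Rigs, then bone corrections and AnimDynamics coat and zipper, into a cached pose.
# bRigidBodyEnabled selects between rigid-body collar and straps with collar correction, or strap trail controllers.
# The update event derives blend alphas from joint rotations, LOD and movement, plus state-based gravity for the straps.
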AnimBlueprint /BRCosmetics/Characters/Player/Female/Medium/Bodies/F_MED_Wasteland_Warrior_Dark/Meshes/F_MED_Wasteland_Warrior_Dark_AnimBP
target_skeleton=&/F_MED_Wasteland_Warrior_Dark_Skeleton
graphs [
 AnimGraph {nodes [
   n1 = RigidBody(ComponentPose=Local To Component(LocalPose=Use cached pose 'basePose'())) {
    bClampLinearTranslationLimitToRefPose=true
    bOverrideWorldGravity=true
    overridePhysicsAsset=&/F_MED_Wasteland_Warrior_Dark_dyn_collar
    simulationSpace=BaseBoneSpace
    alphaBoolBlend {blendOption=Linear}
    baseBoneRef {boneName=spine_05}
    componentAppliedLinearAccClamp=<3000 3000 3000>
    componentLinearAccScale=<0.2 0.2 0.2>
    componentLinearVelScale=<0.2 0.2 0.2>
    overrideWorldGravity {z=980}
   }
   n2 = RigidBody!(ComponentPose=@n1, Alpha, OverrideWorldGravity:Vector=Get GravityOverride(), ExternalForce) -> (Pose) {
    bClampLinearTranslationLimitToRefPose=true
    bOverrideWorldGravity=true
    overridePhysicsAsset=&/F_MED_Wasteland_Warrior_Dark_dyn_straps
    simulationSpace=BaseBoneSpace
    alphaBoolBlend {blendOption=Linear}
    baseBoneRef {boneName=root}
    componentAppliedLinearAccClamp=<100 100 500>
    componentLinearAccScale=<0.1 0.2 0.2>
    componentLinearVelScale=<0.1 0.2 0.2>
    overrideWorldGravity {z=-980}
    simSpaceSettings {
     maxAngularAcceleration=100
     maxAngularVelocity=1000
     maxLinearAcceleration=300
     maxLinearVelocity=5
     velocityScaleZ=0
     externalLinearDragV {z=1}
    }
   }
   n3 = Get head_rot_bk() -> (head_rot_bk:double)
   n4 = "Transform (Modify) Bone - Bone: dyn_collar_bk_MSTR"(ComponentPose=@n2, Alpha=@n3, Translation="0,0,0", Rotation="0,0,0", Scale="0.6,1,1") {
    rotationMode=BMM_Additive
    rotationSpace=BCS_BoneSpace
    scaleMode=BMM_Additive
    scaleSpace=BCS_BoneSpace
    translationMode=BMM_Additive
    translationSpace=BCS_BoneSpace
    alphaBoolBlend {blendOption=Linear}
    boneToModify {boneName=dyn_collar_bk_MSTR}
    scale {x=0.6}
   }
   n5 = "Transform (Modify) Bone - Bone: dyn_collar_bk_r_MSTR"(ComponentPose=@n4, Alpha=@n3, Translation="0,0,0", Rotation="0,0,0", Scale="0.7,1,1") {
    rotationMode=BMM_Additive
    rotationSpace=BCS_BoneSpace
    scaleMode=BMM_Additive
    scaleSpace=BCS_BoneSpace
    translationMode=BMM_Additive
    translationSpace=BCS_BoneSpace
    alphaBoolBlend {blendOption=Linear}
    boneToModify {boneName=dyn_collar_bk_r_MSTR}
    scale {x=0.7}
   }
   n6 = "Transform (Modify) Bone - Bone: dyn_collar_bk_l_MSTR"(ComponentPose=@n5, Alpha=@n3, Translation="0,0,0", Rotation="0,0,0", Scale="0.7,1,1") {
    rotationMode=BMM_Additive
    rotationSpace=BCS_BoneSpace
    scaleMode=BMM_Additive
    scaleSpace=BCS_BoneSpace
    translationMode=BMM_Additive
    translationSpace=BCS_BoneSpace
    alphaBoolBlend {blendOption=Linear}
    boneToModify {boneName=dyn_collar_bk_l_MSTR}
    scale {x=0.7}
   }
   n7 = Get BwdMovement() -> (BwdMovement:double)
   n8 = Trail controller - Bone: dyn_straps_bk_04(ComponentPose=Local To Component(LocalPose=Use cached pose 'basePose'()), Alpha=@n7) {
    chainLength=4
    alphaScaleBiasClamp {bInterpResult=true}
    rotationLimits [{limitMax=<180 180 180>}, {limitMax=<180 180 180>}, {limitMax=<180 180 180>}, {limitMax=<180 180 180>}]
    rotationOffsets [<0 0 0>, <0 0 0>, <0 0 0>, <0 0 0>]
    trailBone {boneName=dyn_straps_bk_04}
    trailRelaxationSpeed {editorCurveData {keys [
       <at 0 35 RCIM_Linear>
       {arriveTangent=-10.9625, interpMode=RCIM_Cubic, leaveTangent=-10.9625, tangentMode=RCTM_User, time=0.2, value=35}
       {arriveTangent=0.8309, interpMode=RCIM_Cubic, leaveTangent=0.8309, tangentMode=RCTM_User, time=1, value=15}
      ]}}
   }
   n9 = Trail controller - Bone: dyn_straps_bk_l_03(ComponentPose=@n8, Alpha=@n7) {
    chainLength=3
    alphaScaleBiasClamp {bInterpResult=true}
    rotationLimits [{limitMax=<180 180 180>}, {limitMax=<180 180 180>}, {limitMax=<180 180 180>}]
    rotationOffsets [<0 0 0>, <0 0 0>, <0 0 0>]
    trailBone {boneName=dyn_straps_bk_l_03}
    trailRelaxationSpeed {editorCurveData {keys [
       <at 0 30 RCIM_Linear>
       {arriveTangent=-0.2592, interpMode=RCIM_Cubic, leaveTangent=-0.2593, tangentMode=RCTM_User, time=0.2, value=30}
       {interpMode=RCIM_Cubic, tangentMode=RCTM_User, time=1, value=15}
      ]}}
   }
   n10 = Get FwdMovement() -> (FwdMovement:double)
   n11 = Trail controller - Bone: dyn_straps_ft_r_02(ComponentPose=@n9, Alpha=@n10) {
    alphaScaleBiasClamp {bInterpResult=true}
    rotationLimits [{limitMax=<180 180 180>}, {limitMax=<180 180 180>}]
    rotationOffsets [<0 0 0>, <0 0 0>]
    trailBone {boneName=dyn_straps_ft_r_02}
    trailRelaxationSpeed {editorCurveData {keys [<at 0 25 RCIM_Linear>, {arriveTangent=0.8309, interpMode=RCIM_Cubic, leaveTangent=0.8309, tangentMode=RCTM_User, time=1, value=15}]}}
   }
   n12 = Trail controller - Bone: dyn_straps_ft_l_02(ComponentPose=@n11, Alpha=@n10) {
    alphaScaleBiasClamp {bInterpResult=true}
    rotationLimits [{limitMax=<180 180 180>}, {limitMax=<180 180 180>}]
    rotationOffsets [<0 0 0>, <0 0 0>]
    trailBone {boneName=dyn_straps_ft_l_02}
    trailRelaxationSpeed {editorCurveData {keys [<at 0 35 RCIM_Linear>, {arriveTangent=0.8309, interpMode=RCIM_Cubic, leaveTangent=0.8309, tangentMode=RCTM_User, time=1, value=25}]}}
   }
   n13 = Blend Poses by bool(BlendPose_0=Component To Local(ComponentPose=@n6), BlendPose_1=Component To Local(ComponentPose=@n12), bActiveValue=Get bRigidBodyEnabled()) {blendType=Linear}
   Output Pose(Result=Component To Local(ComponentPose=Local To Component(LocalPose=@n13))) {layerGroup=None}
   n14 = Control Rig(Source=Copy Pose From Mesh() {bUseAttachedParent=true}) {
    alphaBoolBlend {blendOption=Linear}
    controlRigAssetReference {blueprintRigClass=&/F_MED_WastelandWarrior_Deform_CtrlRig_twists.F_MED_WastelandWarrior_Deform_CtrlRig_twists_C}
   }
   n15 = Control Rig!(
    Source=@n14
    Alpha
    KneeRot_fr_l_114deg:Vector
    KneeTrans_fr_l_114deg:Vector="0.258,-3.591,0.009"
    KneeScale_fr_l_114deg:Vector="1,1,1"
    KneeScale_fr_l_130deg:Vector="1,1,1"
    KneeRot_fr_l_130deg:Vector="0,0,-51.2"
    KneeTrans_fr_l_130deg:Vector="0.258,-3.591,0.009"
   ) -> (Pose) {
    alphaBoolBlend {blendOption=Linear}
    controlRigAssetReference {blueprintRigClass=&/F_MED_WastelandWarrior_Deform_CtrlRig_sculpters.F_MED_WastelandWarrior_Deform_CtrlRig_sculpters_C}
   }
   n16 = "Transform (Modify) Bone - Bone: dyn_glute_l"(
    ComponentPose=Local To Component(LocalPose=@n15)
    Alpha=Get thigh_rot_up_l() -> (thigh_rot_up_l:double)
    Translation="-12,2,0"
    Rotation="0,0,0"
    Scale="1,1,1"
   ) {
    translationMode=BMM_Additive
    translationSpace=BCS_BoneSpace
    alphaBoolBlend {blendOption=Linear}
    boneToModify {boneName=dyn_glute_l}
    translation=<-12 2>
   }
   n17 = "Transform (Modify) Bone - Bone: dyn_glute_r"(ComponentPose=@n16, Alpha=Get thigh_rot_up_r() -> (thigh_rot_up_r:double), Translation="-12,1,0", Rotation="0,0,0", Scale="1,1,1") {
    translationMode=BMM_Additive
    translationSpace=BCS_BoneSpace
    alphaBoolBlend {blendOption=Linear}
    boneToModify {boneName=dyn_glute_r}
    translation=<-12 1>
   }
   n18 = Anim Dynamics - Bone: dyn_coat_bk(ComponentPose=@n17, Alpha=1, ExternalForce="0,0,0") {
    angularDampingOverride=0.999
    bGravityOverrideInSimSpace=true
    bOverrideAngularDamping=true
    bUseGravityOverride=true
    simulationSpace=RootRelative
    alphaBoolBlend {blendOption=Linear}
    boundBone {boneName=dyn_coat_bk}
    componentAppliedLinearAccClamp=<3000 3000 3000>
    componentLinearAccScale=<0.2 0.2 0.2>
    componentLinearVelScale=<0.5 0.5 0.5>
    gravityOverride {z=-980}
    physicsBodyDefinitions [{
      sphereCollisionRadius=0
      boundBone {boneName=dyn_coat_bk}
      boxExtents=<10 20 10>
      constraintSetup {angularLimitsMax=<2 10 2>, angularLimitsMin=<-2 -20 -2>, angularTarget=<0 0 0>}
      localJointOffset=<5 0 0>
     }]
    relativeSpaceBone {boneName=spine_04}
   }
   n19 = Anim Dynamics - Bone: dyn_coat_l(ComponentPose=@n18, Alpha=1, ExternalForce="0,0,0") {
    angularDampingOverride=0.999
    bGravityOverrideInSimSpace=true
    bOverrideAngularDamping=true
    bUseGravityOverride=true
    simulationSpace=RootRelative
    alphaBoolBlend {blendOption=Linear}
    boundBone {boneName=dyn_coat_l}
    componentAppliedLinearAccClamp=<3000 3000 3000>
    componentLinearAccScale=<0.2 0.2 0.2>
    componentLinearVelScale=<0.5 0.5 0.5>
    gravityOverride {z=-980}
    physicsBodyDefinitions [{
      sphereCollisionRadius=0
      boundBone {boneName=dyn_coat_l}
      boxExtents=<10 20 10>
      constraintSetup {angularLimitsMax=<2 10 2>, angularLimitsMin=<-2 -10 -2>, angularTarget=<0 0 0>}
      localJointOffset=<2 0 0>
     }]
    relativeSpaceBone {boneName=spine_04}
   }
   n20 = Anim Dynamics - Bone: dyn_coat_bk_l(ComponentPose=@n19, Alpha=1, ExternalForce="0,0,0") {
    angularDampingOverride=0.999
    bGravityOverrideInSimSpace=true
    bOverrideAngularDamping=true
    bUseGravityOverride=true
    simulationSpace=RootRelative
    alphaBoolBlend {blendOption=Linear}
    boundBone {boneName=dyn_coat_bk_l}
    componentAppliedLinearAccClamp=<3000 3000 3000>
    componentLinearAccScale=<0.2 0.2 0.2>
    componentLinearVelScale=<0.5 0.5 0.5>
    gravityOverride {z=-980}
    physicsBodyDefinitions [{
      sphereCollisionRadius=0
      boundBone {boneName=dyn_coat_bk_l}
      boxExtents=<10 20 10>
      constraintSetup {angularLimitsMax=<2 10 2>, angularLimitsMin=<-2 -10 -2>, angularTarget=<0 0 0>}
      localJointOffset=<4 0 0>
     }]
    relativeSpaceBone {boneName=spine_04}
   }
   n21 = Anim Dynamics - Bone: dyn_coat_ft_l(ComponentPose=@n20, Alpha=1, ExternalForce="0,0,0") {
    angularDampingOverride=0.9995
    bGravityOverrideInSimSpace=true
    bOverrideAngularDamping=true
    bUseGravityOverride=true
    simulationSpace=RootRelative
    alphaBoolBlend {blendOption=Linear}
    boundBone {boneName=dyn_coat_ft_l}
    componentAppliedLinearAccClamp=<3000 3000 3000>
    componentLinearAccScale=<0.2 0.2 0.2>
    componentLinearVelScale=<0.5 0.5 0.5>
    gravityOverride {z=-980}
    physicsBodyDefinitions [{
      sphereCollisionRadius=0
      boundBone {boneName=dyn_coat_ft_l}
      boxExtents=<10 20 10>
      constraintSetup {angularLimitsMax=<2 10 2>, angularLimitsMin=<-2 -10 -2>, angularTarget=<0 0 0>}
      localJointOffset=<3 0 0>
     }]
    relativeSpaceBone {boneName=spine_04}
   }
   n22 = Anim Dynamics - Bone: dyn_coat_r(ComponentPose=@n21, Alpha=1, ExternalForce="0,0,0") {
    angularDampingOverride=0.999
    bGravityOverrideInSimSpace=true
    bOverrideAngularDamping=true
    bUseGravityOverride=true
    simulationSpace=RootRelative
    alphaBoolBlend {blendOption=Linear}
    boundBone {boneName=dyn_coat_r}
    componentAppliedLinearAccClamp=<3000 3000 3000>
    componentLinearAccScale=<0.2 0.2 0.2>
    componentLinearVelScale=<0.5 0.5 0.5>
    gravityOverride {z=-980}
    physicsBodyDefinitions [{
      sphereCollisionRadius=0
      boundBone {boneName=dyn_coat_r}
      boxExtents=<10 20 10>
      constraintSetup {angularLimitsMax=<2 10 2>, angularLimitsMin=<-2 -10 -2>, angularTarget=<0 0 0>}
      localJointOffset=<2 0 0>
     }]
    relativeSpaceBone {boneName=spine_04}
   }
   n23 = Anim Dynamics - Bone: dyn_coat_bk_r(ComponentPose=@n22, Alpha=1, ExternalForce="0,0,0") {
    angularDampingOverride=0.999
    bGravityOverrideInSimSpace=true
    bOverrideAngularDamping=true
    bUseGravityOverride=true
    simulationSpace=RootRelative
    alphaBoolBlend {blendOption=Linear}
    boundBone {boneName=dyn_coat_bk_r}
    componentAppliedLinearAccClamp=<3000 3000 3000>
    componentLinearAccScale=<0.2 0.2 0.2>
    componentLinearVelScale=<0.5 0.5 0.5>
    gravityOverride {z=-980}
    physicsBodyDefinitions [{
      sphereCollisionRadius=0
      boundBone {boneName=dyn_coat_bk_r}
      boxExtents=<10 20 10>
      constraintSetup {angularLimitsMax=<2 10 2>, angularLimitsMin=<-2 -10 -2>, angularTarget=<0 0 0>}
      localJointOffset=<4 0 0>
     }]
    relativeSpaceBone {boneName=spine_04}
   }
   n24 = Anim Dynamics - Bone: dyn_coat_ft_r(ComponentPose=@n23, Alpha=1, ExternalForce="0,0,0") {
    angularDampingOverride=0.9995
    bGravityOverrideInSimSpace=true
    bOverrideAngularDamping=true
    bUseGravityOverride=true
    simulationSpace=RootRelative
    alphaBoolBlend {blendOption=Linear}
    boundBone {boneName=dyn_coat_ft_r}
    componentAppliedLinearAccClamp=<3000 3000 3000>
    componentLinearAccScale=<0.2 0.2 0.2>
    componentLinearVelScale=<0.5 0.5 0.5>
    gravityOverride {z=-980}
    physicsBodyDefinitions [{
      sphereCollisionRadius=0
      boundBone {boneName=dyn_coat_ft_r}
      boxExtents=<10 20 10>
      constraintSetup {angularLimitsMax=<2 10 2>, angularLimitsMin=<-2 -10 -2>, angularTarget=<0 0 0>}
      localJointOffset=<3 0 0>
     }]
    relativeSpaceBone {boneName=spine_04}
   }
   n25 = Anim Dynamics - Bone: dyn_zipper(ComponentPose=@n24, Alpha=1) {
    angularDampingOverride=0.999
    bGravityOverrideInSimSpace=true
    bOverrideAngularDamping=true
    bUseGravityOverride=true
    simulationSpace=RootRelative
    alphaBoolBlend {blendOption=Linear}
    boundBone {boneName=dyn_zipper}
    componentAppliedLinearAccClamp=<3000 3000 3000>
    componentLinearAccScale=<0.2 0.2 0.2>
    componentLinearVelScale=<0.5 0.5 0.5>
    gravityOverride {z=-980}
    physicsBodyDefinitions [{
      sphereCollisionRadius=0
      boundBone {boneName=dyn_zipper}
      boxExtents=<10 5 10>
      constraintSetup {angularLimitsMax=<0 0 35>, angularLimitsMin=<0 0 -35>, angularTarget=<0 0 0>}
      localJointOffset=<3 0 0>
     }]
    relativeSpaceBone {boneName=spine_04}
   }
   n26 = "Constraint - dyn_kneeArmor_r (PP)"!(ComponentPose=@n25, Alpha=1, ConstraintWeights_0=0.9, ConstraintWeights_1:float=0.1) -> (Pose) {
    alphaBoolBlend {blendOption=Linear}
    boneToModify {boneName=dyn_kneeArmor_r}
    constraintSetup [{targetBone {boneName=thigh_r}}, {targetBone {boneName=calf_r}}]
    constraintWeights [1,1]
   }
   n27 = "Transform (Modify) Bone - Bone: dyn_kneeArmor_l"(ComponentPose=@n26, Alpha=Get knee_rot_l() -> (knee_rot_l:double), Translation="0,4,0", Rotation="0.000,-5.000,0.000", Scale="1,1,1") {
    rotationMode=BMM_Additive
    rotationSpace=BCS_BoneSpace
    translationMode=BMM_Additive
    translationSpace=BCS_BoneSpace
    alphaBoolBlend {blendOption=Linear}
    boneToModify {boneName=dyn_kneeArmor_l}
    rotation {yaw=-5.0001}
    translation {y=4}
   }
   n28 = Look At - Bone: dyn_collar_bk_MSTR(ComponentPose=@n27, Alpha=1) {
    alphaBoolBlend {blendOption=Linear}
    boneToModify {boneName=dyn_collar_bk_MSTR}
    lookAtLocation=<30 -5>
    lookAtTarget {boneReference {boneName=neck_02}}
    lookAt_Axis {axis=<1 0>}
    lookUp_Axis {axis=<0 1>}
   }
   n29 = Look At - Bone: dyn_collar_bk_r_MSTR(ComponentPose=@n28, Alpha=1) {
    alphaBoolBlend {blendOption=Linear}
    boneToModify {boneName=dyn_collar_bk_r_MSTR}
    lookAtLocation=<30 6 12>
    lookAtTarget {boneReference {boneName=neck_02}}
    lookAt_Axis {axis=<1 0>}
    lookUp_Axis {axis=<0 1>}
   }
   n30 = "Transform (Modify) Bone - Bone: dyn_straps_ft_MSTR"(ComponentPose=@n29, Alpha=Get thigh_rot_up_l(), Translation="0,0.5,0", Rotation="0,0,0", Scale="1,1,1") {
    translationMode=BMM_Additive
    translationSpace=BCS_BoneSpace
    alphaBoolBlend {blendOption=Linear}
    boneToModify {boneName=dyn_straps_ft_MSTR}
    translation {y=0.5}
   }
   n31 = "Transform (Modify) Bone - Bone: dyn_chest_offset"(
    ComponentPose=@n30
    Alpha=Get chestOffsetAlpha() -> (chestOffsetAlpha:double)
    Translation="-1,0,0"
    Rotation="0,0,0"
    Scale="1.000,0.8,1.000"
   ) {
    scaleMode=BMM_Additive
    scaleSpace=BCS_BoneSpace
    translationMode=BMM_Additive
    translationSpace=BCS_BoneSpace
    alphaBoolBlend {blendOption=Linear}
    boneToModify {boneName=dyn_chest_offset}
    scale {y=0.8}
    translation {x=-1}
   }
   Save cached pose 'basePose'(Pose=Component To Local(ComponentPose=@n31)) {cacheName=basePose}
  ]}
 EventGraph {nodes [
   Event BlueprintUpdateAnimation()
   > n1 = Sequence!(execute) -> (then_0, then_1, then_2:Exec, then_3:Exec, then_4:Exec)
   n2 = ComplexSimSpaceFunction DEPRECATED!(
    execute=@n1.then_2
    CharacterPartAnimInstance:Custom Character Part Anim Instance Object Reference=Self-Reference()
    InterpSpeedMax:double=9
    PawnSpeedMax:double=20
    RInterpErrorTolerance:double=1
    DeltaTime:double=0
    DebugDraw:bool=false
    sim_space_rotation_in:Rotator=Get sim_space_rotation() -> (sim_space_rotation:Rotator)
    inComponentSpaceIn:bool=Get InComponentSpace() -> (InComponentSpace:bool)
    inComponentSpaceLastFrameIn:bool=Get InComponentSpaceLastFrame() -> (InComponentSpaceLastFrame:bool)
    lockedOnSimSpaceIn:bool=Get LockedOnSimSpace() -> (LockedOnSimSpace:bool)
    InterpSpeedIn:double=Get InterpSpeed() -> (InterpSpeed:double)
   ) -> (
    then
    sim_space_rotation:Rotator
    inComponentSpaceOut:bool
    inComponentSpaceLastFrameOut:bool
    lockedOnSimSpaceOut:bool
    InterpSpeedOut:double
   )
   > Set sim_space_rotation(sim_space_rotation:Rotator=@n2) -> (Output_Get:Rotator)
   > Set InComponentSpace(InComponentSpace:bool=@n2.inComponentSpaceOut) -> (Output_Get:bool)
   > Set InComponentSpaceLastFrame(InComponentSpaceLastFrame:bool=@n2.inComponentSpaceLastFrameOut) -> (Output_Get:bool)
   > Set LockedOnSimSpace(LockedOnSimSpace:bool=@n2.lockedOnSimSpaceOut) -> (Output_Get:bool)
   > Set InterpSpeed(InterpSpeed:double=@n2.InterpSpeedOut) -> (Output_Get:double)
   n3 = Gravity Override State Function DEPRECATED(
    execute=@n1.then_3
    CharacterPartAnimInstance=Self-Reference()
    GameTimeInSeconds=GetGameTimeInSeconds()
    GravityOverrideParams=Get Gravity Override Params()
   )
   > Set GravityOverride(GravityOverride=@n3)
   n4 = PlayerMovement(execute=@n1.then_4, AnimInstance=Self-Reference(), XDotProductCoefficient=1, YDotProductCoefficient=1, ZDotProductCoefficient=1)
   n5 = MapRangeClamped(Value=@n4.YMovement, InRangeA=0, InRangeB=1)
   > Set FwdMovement(FwdMovement:double=MapRangeClamped(
     Value=float + float(A=@n4, B=@n5) {is=title K2Node_CommutativeAssociativeBinaryOperator}
     InRangeA=0
     InRangeB=1.1
     OutRangeA=0.5
     OutRangeB=0
    )) -> (Output_Get:double)
   n6 = MapRangeClamped(Value=@n4.YMovement, InRangeA=0, InRangeB=-1)
   n7 = "Absolute (Float)"(A=@n4.ZMovement)
   > Set BwdMovement(BwdMovement:double=MapRangeClamped(Value=float + float(A=@n7, B=@n6), InRangeA=0, InRangeB=1.5, OutRangeA=1, OutRangeB=0)) -> (Output_Get:double)
   n8 = BreakTransform(InTransform=GetDeltaTransformFromRefPose(self=GetOwnerSkeletalMeshComponent(), BoneName=thigh_r, BaseName=pelvis))
   Set thigh_rot_up_r(execute=@n1.then_1, thigh_rot_up_r:double=MapRangeClamped(Value=BreakRotator(InRot=@n8.Rotation).Yaw, InRangeA=-15, InRangeB=-100)) -> (Output_Get:double)
   n9 = BreakTransform(InTransform=GetDeltaTransformFromRefPose(self=GetOwnerSkeletalMeshComponent(), BoneName=thigh_l, BaseName=pelvis))
   > Set thigh_rot_up_l(thigh_rot_up_l:double=MapRangeClamped(Value=BreakRotator(InRot=@n9.Rotation).Yaw, InRangeA=-15, InRangeB=-100)) -> (Output_Get:double)
   n10 = Get CurrentLOD()
   > Set chestOffsetAlpha(chestOffsetAlpha:double="To Float (Boolean)"(InBool="integer >= integer"!(A:int=@n10, B:int=1) -> (ReturnValue:bool) {is=title K2Node_CallFunction})) -> (Output_Get:double)
   n11 = BreakTransform(InTransform=GetDeltaTransformFromRefPose(self=GetOwnerSkeletalMeshComponent(), BoneName=calf_l, BaseName=thigh_l))
   Set knee_rot_l(execute=@n1.then_0, knee_rot_l:double=MapRangeClamped(Value=BreakRotator(InRot=@n11.Rotation).Yaw, InRangeA=10, InRangeB=110)) -> (Output_Get:double)
   n12 = BreakTransform(InTransform=GetDeltaTransformFromRefPose(self=GetOwnerSkeletalMeshComponent(), BoneName=head, BaseName=spine_05))
   > Set head_rot_bk(head_rot_bk:double=MapRangeClamped(Value=BreakRotator(InRot=@n12.Rotation).Yaw, InRangeA=0, InRangeB=-50)) -> (Output_Get:double)
  ]}
]
variables [
 BwdMovement {type=number}
 chestOffsetAlpha {type=number}
 FwdMovement {type=number}
 Gravity Override Params {
  type=GravityOverrideParamsStruct
  default {
   skydive_FloatingDown {gravityOverride=<0 -700 100>, windAmplitude=<200 -500 200>, windOffset=<0 -600 200>}
   skydive_GoingUp {gravityOverride=<0 -300 500>, windAmplitude=<200 -500 300>}
   skydive_Parachuting {gravityOverride=<0 -500 -980>, windAmplitude=<200 -300 400>}
  }
 }
 GravityOverride {type=Vector}
 head_rot_bk {type=number}
 InComponentSpace {type=boolean}
 InComponentSpaceLastFrame {type=boolean}
 InterpSpeed {type=number}
 knee_rot_l {type=number}
 LockedOnSimSpace {type=boolean}
 sim_space_rotation {type=Rotator}
 thigh_rot_up_l {type=number}
 thigh_rot_up_r {type=number}
]
