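# Physics asset for the dynamic hair and braids of the female medium Wasteland Warrior Dark cosmetic character.
# Capsules on the upper spine, neck and clavicles form the body volume, joined by angularly free constraints.
# A back-hair chain hangs from a free-swinging master bone on the neck, with swing limits mostly widening toward the tip and tight twist.
# Braid capsules branch from the first two back-hair segments with narrower, uniform limits.
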PhysicsAsset /BRCosmetics/Characters/Player/Female/Medium/Bodies/F_MED_Wasteland_Warrior_Dark/Meshes/Parts/F_MED_Wasteland_Warrior_Dark_hair_dyn
bodies {
 clavicle_l [{shapeType=Capsule, length=10.25, radius=7.375, center=<4.5449 0.5263 0.5146>, rotation=<rot 84.9994 0 -5>}]
 clavicle_r [{shapeType=Capsule, length=10.125, radius=7.5, center=<-4.6488 0.2438 -0.6463>, rotation=<rot 84.9992 0 -174.9998>}]
 dyn_hairBraid_00 [{shapeType=Capsule, length=3.6871, radius=1.2604, center=<3 0 0>, rotation=<rot 0 91.5686 -89.2139>}]
 dyn_hairBraid_01 [{shapeType=Capsule, length=2.8121, radius=1.2604, center=<3.7951 0 0>, rotation=<rot 0 91.5686 -89.2139>}]
 dyn_hairBraid_02 [{shapeType=Capsule, length=2.2197, radius=1.2796, center=<4.2365 -0.5 0.5>, rotation=<rot 0 -87.8632 -84.8611>}]
 dyn_hairBraid_03 [{shapeType=Capsule, length=2.3447, radius=1.1546, center=<3 -0.5 0.5>, rotation=<rot 0 -87.8632 -84.8611>}]
 dyn_hair_bk_01 [{shapeType=Capsule, length=2.75, radius=2.25, center=<3 0 0>, rotation=<rot 90 0 0>}]
 dyn_hair_bk_02 [{shapeType=Capsule, length=2, radius=2.5, center=<3 0 0>, rotation=<rot 90 0 0>}]
 dyn_hair_bk_03 [{shapeType=Capsule, length=3.25, radius=2.5, center=<2.9808 0 -0.2197>, rotation=<rot 85 0 0>}]
 dyn_hair_bk_04 [{shapeType=Capsule, length=1.875, radius=2.5, center=<2.9612 0 -0.6525>, rotation=<rot 0 0 90>}]
 dyn_hair_bk_MSTR [{shapeType=Capsule, length=0.505, radius=0.505, rotation=<rot 0 -90 0>}]
 neck_01 [{shapeType=Capsule, length=3.875, radius=5.875, center=<3.5972 0.9993 0.0001>, rotation=<rot 90 -90 -85>}]
 spine_05 [{shapeType=Capsule, length=16.75, radius=10.25, center=<2.7582 1.4383 0.0075>, rotation=<rot 0 0 -5.0006>}]
}
constraints [
 {bone1Name=dyn_hair_bk_01, bone2Name=dyn_hair_bk_MSTR, swing1LimitDegrees=20, swing2LimitDegrees=10, twistLimitDegrees=2}
 {bone1Name=dyn_hair_bk_02, bone2Name=dyn_hair_bk_01, swing1LimitDegrees=20, swing2LimitDegrees=25, twistLimitDegrees=2}
 {bone1Name=dyn_hair_bk_03, bone2Name=dyn_hair_bk_02, swing1LimitDegrees=30, swing2LimitDegrees=30, twistLimitDegrees=2}
 {
  bone1Name=dyn_hair_bk_MSTR
  bone2Name=neck_01
  swing1LimitDegrees=45
  swing1Motion=Free
  swing2LimitDegrees=45
  swing2Motion=Free
  twistLimitDegrees=45
  twistMotion=Free
 }
 {
  bone1Name=neck_01
  bone2Name=spine_05
  swing1LimitDegrees=45
  swing1Motion=Free
  swing2LimitDegrees=45
  swing2Motion=Free
  twistLimitDegrees=45
  twistMotion=Free
 }
 {bone1Name=dyn_hairBraid_01, bone2Name=dyn_hairBraid_00, swing1LimitDegrees=15, swing2LimitDegrees=15, twistLimitDegrees=2}
 {bone1Name=dyn_hairBraid_02, bone2Name=dyn_hair_bk_02, swing1LimitDegrees=15, swing2LimitDegrees=15, twistLimitDegrees=2}
 {bone1Name=dyn_hair_bk_04, bone2Name=dyn_hair_bk_03, swing1LimitDegrees=30, swing2LimitDegrees=30, twistLimitDegrees=2}
 {bone1Name=dyn_hairBraid_00, bone2Name=dyn_hair_bk_01, swing1LimitDegrees=15, swing2LimitDegrees=15, twistLimitDegrees=2}
 {bone1Name=dyn_hairBraid_03, bone2Name=dyn_hair_bk_01, swing1LimitDegrees=15, swing2LimitDegrees=15, twistLimitDegrees=2}
 {
  bone1Name=clavicle_l
  bone2Name=spine_05
  swing1LimitDegrees=45
  swing1Motion=Free
  swing2LimitDegrees=45
  swing2Motion=Free
  twistLimitDegrees=45
  twistMotion=Free
 }
 {
  bone1Name=clavicle_r
  bone2Name=spine_05
  swing1LimitDegrees=45
  swing1Motion=Free
  swing2LimitDegrees=45
  swing2Motion=Free
  twistLimitDegrees=45
  twistMotion=Free
 }
]
