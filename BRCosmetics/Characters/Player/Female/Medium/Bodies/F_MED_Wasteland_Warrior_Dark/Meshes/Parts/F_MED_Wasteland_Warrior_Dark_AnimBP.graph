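# Animation Blueprint for the Wasteland Warrior Dark female body part, adding secondary hair motion on a pose copied from the attached parent mesh.
# The back hair master bone aims at spine_05, then shifts additively by an amount driven by head yaw relative to spine_05.
# A neck-relative RigidBody simulation follows, its gravity updated each frame by a deprecated state function with skydive-specific gravity and wind settings.
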AnimBlueprint /BRCosmetics/Characters/Player/Female/Medium/Bodies/F_MED_Wasteland_Warrior_Dark/Meshes/Parts/F_MED_Wasteland_Warrior_Dark_AnimBP
target_skeleton=&/F_MED_Wasteland_Warrior_Dark_Skeleton
graphs [
 AnimGraph {nodes [
   n1 = Look At - Bone: dyn_hair_bk_MSTR(ComponentPose=Local To Component(LocalPose=Copy Pose From Mesh() {bUseAttachedParent=true}), Alpha=1) {
    alphaBoolBlend {blendOption=Linear}
    boneToModify {boneName=dyn_hair_bk_MSTR}
    lookAtLocation=<0 -8>
    lookAtTarget {boneReference {boneName=spine_05}}
    lookAt_Axis {axis=<1 0>}
    lookUp_Axis {axis {x=0}}
   }
   n2 = "Transform (Modify) Bone - Bone: dyn_hair_bk_MSTR"(ComponentPose=@n1, Alpha=Get head_rot_bkw() -> (head_rot_bkw:double), Translation="0.000,0.000,-2", Rotation="0,0,0", Scale="1,1,1") {
    rotationMode=BMM_Additive
    rotationSpace=BCS_BoneSpace
    translationMode=BMM_Additive
    translationSpace=BCS_BoneSpace
    alphaBoolBlend {blendOption=Linear}
    boneToModify {boneName=dyn_hair_bk_MSTR}
    translation {z=-2}
   }
   n3 = RigidBody!(ComponentPose=@n2, Alpha, OverrideWorldGravity:Vector=Get GravityOverride(), ExternalForce) -> (Pose) {
    bClampLinearTranslationLimitToRefPose=true
    bOverrideWorldGravity=true
    overridePhysicsAsset=&/F_MED_Wasteland_Warrior_Dark_hair_dyn
    simulationSpace=BaseBoneSpace
    alphaBoolBlend {blendOption=Linear}
    baseBoneRef {boneName=neck_01}
    componentAppliedLinearAccClamp=<3000 3000 3000>
    componentLinearAccScale=<0.3 0.3 0.3>
    componentLinearVelScale=<0.2 0.2 0.2>
    overrideWorldGravity {z=-980}
   }
   Output Pose(Result=Component To Local(ComponentPose=@n3)) {layerGroup=None}
  ]}
 EventGraph {nodes [
   Event BlueprintUpdateAnimation()
   n1 = BreakTransform(InTransform=GetDeltaTransformFromRefPose(self=GetOwnerSkeletalMeshComponent(), BoneName=head, BaseName=spine_05))
   > Set head_rot_bkw(head_rot_bkw:double=MapRangeClamped(Value=BreakRotator(InRot=@n1.Rotation).Yaw, InRangeA=0, InRangeB=-50)) -> (Output_Get:double)
   > n2 = Gravity Override State Function DEPRECATED(CharacterPartAnimInstance=Self-Reference(), GameTimeInSeconds=GetGameTimeInSeconds(), GravityOverrideParams=Get Gravity Override Params())
   > Set GravityOverride(GravityOverride=@n2)
  ]}
]
variables [
 Gravity Override Params {
  type=GravityOverrideParamsStruct
  default {
   skydive_DivingDown {windFrequency=<1 6 5>}
   skydive_FloatingDown {gravityOverride=<0 -700 -300>, windAmplitude=<100 -300 100>, windFrequency=<1 3 4>, windOffset=<0 -300 100>}
   skydive_GoingUp {gravityOverride=<0 -300 500>, windAmplitude=<200 -500 300>}
   skydive_Parachuting {gravityOverride=<0 -300 -980>, windAmplitude=<100 -200 200>, windFrequency=<1.5 4 6>}
  }
 }
 GravityOverride {type=Vector}
 head_rot_bkw {type=number}
]
